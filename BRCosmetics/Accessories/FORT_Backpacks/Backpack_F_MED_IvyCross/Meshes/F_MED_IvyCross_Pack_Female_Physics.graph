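# Physics asset for the female IvyCross backpack cosmetic, setting up physics for its dangling beads, ribbon and tassels.
# Two constrained chains hang from the pack: lower bead to tassel to middle tassel, and upper bead to ribbon, each with per-joint swing and twist limits.
# Capsules on the pack, pelvis and lower spine approximate backpack and body volumes, with the spine_01 to spine_02 joint left free in swing and twist.
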PhysicsAsset /BRCosmetics/Accessories/FORT_Backpacks/Backpack_F_MED_IvyCross/Meshes/F_MED_IvyCross_Pack_Female_Physics
bodies {
 dyn_bead_lower [{shapeType=Capsule, length=0.5, radius=1.75, shapeName=dyn_bead_lower_capsule, center=<2.083 0 0>, rotation=<rot 90 160.5289 160.5294>}]
 dyn_bead_upper [{shapeType=Sphere, radius=1.1457, shapeName=dyn_bead_upper_sphere, center=<1.5776 0.0692 0.0369>}]
 dyn_middle_tassel [{shapeType=Sphere, radius=4.75, shapeName=dyn_middle_tassel_sphere, center=<2.643 0 0>}]
 dyn_ribbon [{shapeType=Capsule, length=3.75, radius=0.75, shapeName=dyn_ribbon_capsule, center=<-0.8563 -1.8286 0>, rotation=<rot 90 0 165.001>}]
 dyn_tassel [{shapeType=Sphere, radius=3, shapeName=dyn_tassel_sphere, center=<3.6998 -0.2257 -0.1041>}]
 pack [
  {shapeType=Capsule, length=20.75, radius=12.25, shapeName=pack_capsule, center=<-0.6189 18.2214 0>, rotation=<rot 90 -179.1048 -176.5519>}
  {shapeType=Capsule, length=2.75, radius=3, shapeName=pack_capsule, center=<9.7369 1.5951 0>, rotation=<rot 0 179.9998 -180>}
 ]
 pelvis [{shapeType=Capsule, length=6, radius=13.25, shapeName=pelvis_capsule, center=<0 -0.8394 0>}]
 spine_01 [{shapeType=Capsule, length=5.25, radius=8.5, shapeName=spine_01_capsule, center=<0 -0.7945 0>}]
 spine_02 [{shapeType=Capsule, length=4.5, radius=7.5, shapeName=spine_02_capsule, center=<0 -0.7945 0>}]
}
constraints [
 {bone1Name=dyn_bead_lower, bone2Name=pack, swing1LimitDegrees=40, swing2LimitDegrees=80, twistLimitDegrees=10}
 {bone1Name=dyn_bead_upper, bone2Name=pack, swing1LimitDegrees=15, swing2LimitDegrees=10, twistLimitDegrees=5}
 {bone1Name=dyn_tassel, bone2Name=dyn_bead_lower, swing1LimitDegrees=120, swing2LimitDegrees=120, twistLimitDegrees=15}
 {bone1Name=dyn_ribbon, bone2Name=dyn_bead_upper, swing1LimitDegrees=140, swing2LimitDegrees=60, twistLimitDegrees=1}
 {
  bone1Name=spine_01
  bone2Name=spine_02
  swing1LimitDegrees=45
  swing1Motion=Free
  swing2LimitDegrees=45
  swing2Motion=Free
  twistLimitDegrees=45
  twistMotion=Free
 }
 {bone1Name=dyn_middle_tassel, bone2Name=dyn_tassel, swing1LimitDegrees=25, swing2LimitDegrees=40, twistLimitDegrees=15}
]
