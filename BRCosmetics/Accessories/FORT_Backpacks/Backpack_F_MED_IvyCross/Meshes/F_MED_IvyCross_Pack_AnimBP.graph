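# Animation Blueprint for the IvyCross backpack cosmetic, simulating its dangling ribbon, beads and tassels.
# Each update copies pelvis and spine transforms from the body mesh, then a gender-offset pack pose is cached as Presim.
# Presim feeds RigidBody, AnimDynamics or fallback Trail chains, chosen by enable flags, with gender-specific physics assets, limits or alphas.
# Movement-state gravity blends toward strong downward pull during jumps and falls, outside skydiving.
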AnimBlueprint /BRCosmetics/Accessories/FORT_Backpacks/Backpack_F_MED_IvyCross/Meshes/F_MED_IvyCross_Pack_AnimBP
target_skeleton=&/F_MED_IvyCross_Pack_Skeleton
graphs [
 AnimGraph {nodes [
   n1 = "Transform (Modify) Bone - Bone: dyn_bead_lower"!(
    ComponentPose=Local To Component(LocalPose=Use cached pose 'Presim'())
    bAlphaBoolEnabled:bool=Get bIsJumping()
    Translation="0,0,0"
    Rotation="0,-15,0"
    Scale="1,1,1"
   ) -> (Pose) {
    alphaInputType=Bool
    rotationMode=BMM_Additive
    rotationSpace=BCS_BoneSpace
    alphaBoolBlend {blendInTime=0.3, blendOption=Linear, blendOutTime=0.4}
    boneToModify {boneName=dyn_bead_lower}
    rotation {yaw=-15}
   }
   n2 = Trail controller - Bone: dyn_middle_tassel(ComponentPose=@n1, Alpha=Get Is Female Trail Alpha() -> (Is Female Trail Alpha:float)) {
    bLimitRotation=true
    bUsePlanarLimit=true
    maxDeltaTime=0.0333
    relaxationSpeedScale=2.9
    planarLimits [
     {drivingBone {boneName=spine_01}, planeTransform {location=<0 -10 0>, rotation=<rot 0 0 -90>, scale=<1 1 1>}}
     {drivingBone {boneName=spine_02}, planeTransform {location=<0 -10 0>, rotation=<rot 0 0 -90>, scale=<1 1 1>}}
     {drivingBone {boneName=pelvis}, planeTransform {location=<0 -10 0>, rotation=<rot 0 0 -90>, scale=<1 1 1>}}
    ]
    rotationLimits [{limitMax=<30 190 30>, limitMin=<-30 -10 -60>}, {limitMax=<20 160 20>, limitMin=<-20 20 -20>}]
    rotationOffsets [<0 0 0>, <0 0 0>]
    trailBone {boneName=dyn_middle_tassel}
    trailRelaxationSpeed {editorCurveData {keys [<at 0.0037 12 RCIM_Linear>,<at 0.5 7.5 RCIM_Linear>,<at 1 6 RCIM_Linear>]}}
   }
   n3 = Trail controller - Bone: dyn_middle_tassel(ComponentPose=@n2, Alpha=Get Is Male Trail Alpha() -> (Is Male Trail Alpha:float)) {
    bLimitRotation=true
    bUsePlanarLimit=true
    maxDeltaTime=0.0333
    relaxationSpeedScale=2.9
    planarLimits [
     {drivingBone {boneName=spine_01}, planeTransform {location=<0 -18 0>, rotation=<rot 0 0 -90>, scale=<1 1 1>}}
     {drivingBone {boneName=spine_02}, planeTransform {location=<0 -18 0>, rotation=<rot 0 0 -90>, scale=<1 1 1>}}
     {drivingBone {boneName=pelvis}, planeTransform {location=<0 -18 0>, rotation=<rot 0 0 -90>, scale=<1 1 1>}}
    ]
    rotationLimits [{limitMax=<30 190 30>, limitMin=<-30 -10 -60>}, {limitMax=<20 160 20>, limitMin=<-20 20 -20>}]
    rotationOffsets [<0 0 0>, <0 0 0>]
    trailBone {boneName=dyn_middle_tassel}
    trailRelaxationSpeed {editorCurveData {keys [<at 0.0037 12 RCIM_Linear>,<at 0.5 7.5 RCIM_Linear>,<at 1 6 RCIM_Linear>]}}
   }
   n4 = Trail controller - Bone: dyn_ribbon(ComponentPose=@n3, Alpha=0.3) {
    bActorSpaceFakeVel=true
    bLimitRotation=true
    maxDeltaTime=0.033
    relaxationSpeedScale=2.9
    baseJoint {boneName=root}
    rotationLimits [{limitMax=<20 20 20>, limitMin=<-20 -20 -20>}, {limitMax=<90 90 90>, limitMin=<-90 -90 -90>}]
    rotationOffsets [<0 0 0>, <0 0 0>]
    trailBone {boneName=dyn_ribbon}
    trailRelaxationSpeed {editorCurveData {keys [<at 0.0037 12 RCIM_Linear>,<at 0.4333 7.5171 RCIM_Linear>,<at 1 6 RCIM_Linear>]}}
   }
   n5 = Component To Local(ComponentPose=@n4)
   n6 = Get Gravity Override()
   n7 = Anim Dynamics - Bone: dyn_ribbon!(
    ComponentPose=Local To Component(LocalPose=Use cached pose 'Presim'())
    bAlphaBoolEnabled:bool=True
    GravityOverride:Vector=@n6
    ExternalForce="0,0,0"
   ) -> (Pose) {
    alphaInputType=Bool
    angularBiasOverride=0.999
    angularDampingOverride=0.999
    angularSpringConstant=300
    bOverrideAngularBias=true
    bOverrideAngularDamping=true
    bUseGravityOverride=true
    simulationSpace=RootRelative
    alphaBoolBlend {blendOption=Linear}
    boundBone {boneName=dyn_ribbon}
    componentAppliedLinearAccClamp=<3000 3000 3000>
    componentLinearAccScale=<0.35 0.45 0.45>
    componentLinearVelScale=<0.35 0.45 0.45>
    physicsBodyDefinitions [{
      boundBone {boneName=dyn_ribbon}
      boxExtents=<10 10 10>
      constraintSetup {angularLimitsMax=<1 15 15>, angularLimitsMin=<-1 -15 -15>, angularTarget=<0 0 0>}
      localJointOffset=<1 1 0>
     }]
   }
   n8 = Anim Dynamics - Bone: dyn_bead_upper!(ComponentPose=@n7, bAlphaBoolEnabled:bool=True, GravityOverride:Vector=@n6, ExternalForce="0,0,0") -> (Pose) {
    alphaInputType=Bool
    angularBiasOverride=0.999
    angularDampingOverride=0.999
    angularSpringConstant=150
    bAngularSpring=true
    bOverrideAngularBias=true
    bOverrideAngularDamping=true
    bUseGravityOverride=true
    simulationSpace=RootRelative
    alphaBoolBlend {blendOption=Linear}
    boundBone {boneName=dyn_bead_upper}
    componentAppliedLinearAccClamp=<3000 3000 3000>
    componentLinearAccScale=<0.35 0.45 0.45>
    componentLinearVelScale=<0.35 0.45 0.45>
    physicsBodyDefinitions [{
      boundBone {boneName=dyn_bead_upper}
      boxExtents=<8 8 8>
      constraintSetup {angularLimitsMax=<2.5 5 7.5>, angularLimitsMin=<-2.5 -5 -7.5>, angularTarget=<0 0 0>}
      localJointOffset=<1 0 0>
     }]
   }
   n9 = Get isFemale() -> (isFemale:bool)
   n10 = Get Gravity Override()
   n11 = Anim Dynamics - Bone: dyn_bead_lower!(ComponentPose=@n8, bAlphaBoolEnabled:bool=@n9, GravityOverride:Vector=@n10, ExternalForce="0,0,0") -> (Pose) {
    alphaInputType=Bool
    angularBiasOverride=0.999
    angularDampingOverride=0.999
    angularSpringConstant=1500
    bOverrideAngularBias=true
    bOverrideAngularDamping=true
    bUseGravityOverride=true
    bUsePlanarLimit=false
    bUseSphericalLimits=true
    simulationSpace=RootRelative
    alphaBoolBlend {blendOption=Linear}
    boundBone {boneName=dyn_bead_lower}
    componentAppliedLinearAccClamp=<3000 3000 3000>
    componentLinearAccScale=<0.4 0.5 0.5>
    componentLinearVelScale=<0.4 0.5 0.5>
    physicsBodyDefinitions [{
      collisionType=CustomSphere
      sphereCollisionRadius=2
      boundBone {boneName=dyn_bead_lower}
      boxExtents=<7 4 4>
      constraintSetup {angularLimitsMax=<2.5 50 0>, angularLimitsMin=<-2.5 -50 -10>, angularTarget=<0 0 0>}
      localJointOffset=<1 0 0>
     }]
    planarLimits [{drivingBone {boneName=spine_02}, planeTransform {location=<0 -20 0>, rotation=<rot 0 0 -90>}}]
    sphericalLimits [
     {limitRadius=7, drivingBone {boneName=pelvis}, sphereLocalOffset=<0 0 0>}
     {limitRadius=7, drivingBone {boneName=spine_01}, sphereLocalOffset=<5 0 0>}
     {limitRadius=7, drivingBone {boneName=spine_02}, sphereLocalOffset=<7 0 0>}
    ]
   }
   n12 = Anim Dynamics - Bone: dyn_tassel!(ComponentPose=@n11, bAlphaBoolEnabled:bool=@n9, GravityOverride:Vector=@n10, ExternalForce="0,0,0") -> (Pose) {
    alphaInputType=Bool
    angularBiasOverride=0.999
    angularDampingOverride=0.999
    angularSpringConstant=1500
    bOverrideAngularBias=true
    bOverrideAngularDamping=true
    bUseGravityOverride=true
    bUsePlanarLimit=false
    bUseSphericalLimits=true
    simulationSpace=RootRelative
    alphaBoolBlend {blendOption=Linear}
    boundBone {boneName=dyn_tassel}
    componentAppliedLinearAccClamp=<3000 3000 3000>
    componentLinearAccScale=<0.4 0.5 0.5>
    componentLinearVelScale=<0.4 0.5 0.5>
    physicsBodyDefinitions [{
      collisionType=CustomSphere
      sphereCollisionRadius=2
      boundBone {boneName=dyn_tassel}
      boxExtents=<10 5.5 5.5>
      constraintSetup {angularLimitsMax=<2.5 50 0>, angularLimitsMin=<-2.5 -50 -20>, angularTarget=<0 0 0>}
      localJointOffset=<1.5 0 0>
     }]
    planarLimits [{drivingBone {boneName=spine_02}, planeTransform {location=<0 -20 0>, rotation=<rot 0 0 -90>}}]
    relativeSpaceBone {boneName=dyn_bead_lower}
    sphericalLimits [
     {limitRadius=7, drivingBone {boneName=pelvis}, sphereLocalOffset=<0 0 0>}
     {limitRadius=7, drivingBone {boneName=spine_01}, sphereLocalOffset=<5 0 0>}
     {limitRadius=7, drivingBone {boneName=spine_02}, sphereLocalOffset=<7 0 0>}
    ]
   }
   n13 = Anim Dynamics - Bone: dyn_middle_tassel!(ComponentPose=@n12, bAlphaBoolEnabled:bool=@n9, GravityOverride:Vector=@n10, ExternalForce="0,0,0") -> (Pose) {
    alphaInputType=Bool
    angularBiasOverride=0.999
    angularDampingOverride=0.999
    angularSpringConstant=1500
    bOverrideAngularBias=true
    bOverrideAngularDamping=true
    bUseGravityOverride=true
    bUsePlanarLimit=false
    bUseSphericalLimits=true
    simulationSpace=RootRelative
    alphaBoolBlend {blendOption=Linear}
    boundBone {boneName=dyn_middle_tassel}
    componentAppliedLinearAccClamp=<3000 3000 3000>
    componentLinearAccScale=<0.4 0.5 0.5>
    componentLinearVelScale=<0.4 0.5 0.5>
    physicsBodyDefinitions [{
      collisionType=CustomSphere
      sphereCollisionRadius=4
      boundBone {boneName=dyn_middle_tassel}
      boxExtents=<17 8 8>
      constraintSetup {angularLimitsMax=<2.5 40 0>, angularLimitsMin=<-2.5 -40 -20>, angularTarget=<0 0 0>}
      localJointOffset=<3 0 0>
     }]
    planarLimits [{drivingBone {boneName=spine_02}, planeTransform {location=<0 -20 0>, rotation=<rot 0 0 -90>}}]
    relativeSpaceBone {boneName=dyn_tassel}
    sphericalLimits [
     {limitRadius=7, drivingBone {boneName=pelvis}, sphereLocalOffset=<0 0 0>}
     {limitRadius=7, drivingBone {boneName=spine_01}, sphereLocalOffset=<5 0 0>}
     {limitRadius=7, drivingBone {boneName=spine_02}, sphereLocalOffset=<7 0 0>}
    ]
   }
   n14 = Get Gravity Override()
   n15 = NOT Boolean(A=Get isFemale())
   n16 = Anim Dynamics - Bone: dyn_bead_lower(ComponentPose=@n13, bAlphaBoolEnabled=@n15, GravityOverride=@n14, ExternalForce="0,0,0") {
    alphaInputType=Bool
    angularBiasOverride=0.999
    angularDampingOverride=0.999
    angularSpringConstant=1500
    bOverrideAngularBias=true
    bOverrideAngularDamping=true
    bUseGravityOverride=true
    bUsePlanarLimit=false
    bUseSphericalLimits=true
    simulationSpace=RootRelative
    alphaBoolBlend {blendOption=Linear}
    boundBone {boneName=dyn_bead_lower}
    componentAppliedLinearAccClamp=<3000 3000 3000>
    componentLinearAccScale=<0.4 0.5 0.5>
    componentLinearVelScale=<0.4 0.5 0.5>
    physicsBodyDefinitions [{
      collisionType=CustomSphere
      sphereCollisionRadius=2
      boundBone {boneName=dyn_bead_lower}
      boxExtents=<7 4 4>
      constraintSetup {angularLimitsMax=<2.5 50 0>, angularLimitsMin=<-2.5 -50 -10>, angularTarget=<0 0 0>}
      localJointOffset=<1 0 0>
     }]
    planarLimits [{drivingBone {boneName=spine_02}, planeTransform {location=<0 -20 0>, rotation=<rot 0 0 -90>}}]
    sphericalLimits [
     {limitRadius=13, drivingBone {boneName=pelvis}, sphereLocalOffset=<0 0 0>}
     {limitRadius=13, drivingBone {boneName=spine_01}, sphereLocalOffset=<5 0 0>}
     {limitRadius=13, drivingBone {boneName=spine_02}, sphereLocalOffset=<7 0 0>}
    ]
   }
   n17 = Anim Dynamics - Bone: dyn_tassel(ComponentPose=@n16, bAlphaBoolEnabled=@n15, GravityOverride=@n14, ExternalForce="0,0,0") {
    alphaInputType=Bool
    angularBiasOverride=0.999
    angularDampingOverride=0.999
    angularSpringConstant=1500
    bOverrideAngularBias=true
    bOverrideAngularDamping=true
    bUseGravityOverride=true
    bUsePlanarLimit=false
    bUseSphericalLimits=true
    simulationSpace=RootRelative
    alphaBoolBlend {blendOption=Linear}
    boundBone {boneName=dyn_tassel}
    componentAppliedLinearAccClamp=<3000 3000 3000>
    componentLinearAccScale=<0.4 0.5 0.5>
    componentLinearVelScale=<0.4 0.5 0.5>
    physicsBodyDefinitions [{
      collisionType=CustomSphere
      sphereCollisionRadius=2
      boundBone {boneName=dyn_tassel}
      boxExtents=<10 5.5 5.5>
      constraintSetup {angularLimitsMax=<2.5 50 0>, angularLimitsMin=<-2.5 -50 -20>, angularTarget=<0 0 0>}
      localJointOffset=<1 0 0>
     }]
    planarLimits [{drivingBone {boneName=spine_02}, planeTransform {location=<0 -20 0>, rotation=<rot 0 0 -90>}}]
    relativeSpaceBone {boneName=dyn_bead_lower}
    sphericalLimits [
     {limitRadius=13, drivingBone {boneName=pelvis}, sphereLocalOffset=<0 0 0>}
     {limitRadius=13, drivingBone {boneName=spine_01}, sphereLocalOffset=<5 0 0>}
     {limitRadius=13, drivingBone {boneName=spine_02}, sphereLocalOffset=<7 0 0>}
    ]
   }
   n18 = Anim Dynamics - Bone: dyn_middle_tassel(ComponentPose=@n17, bAlphaBoolEnabled=@n15, GravityOverride=@n14, ExternalForce="0,0,0") {
    alphaInputType=Bool
    angularBiasOverride=0.999
    angularDampingOverride=0.999
    angularSpringConstant=1500
    bOverrideAngularBias=true
    bOverrideAngularDamping=true
    bUseGravityOverride=true
    bUsePlanarLimit=false
    bUseSphericalLimits=true
    simulationSpace=RootRelative
    alphaBoolBlend {blendOption=Linear}
    boundBone {boneName=dyn_middle_tassel}
    componentAppliedLinearAccClamp=<3000 3000 3000>
    componentLinearAccScale=<0.4 0.5 0.5>
    componentLinearVelScale=<0.4 0.5 0.5>
    physicsBodyDefinitions [{
      collisionType=CustomSphere
      sphereCollisionRadius=4
      boundBone {boneName=dyn_middle_tassel}
      boxExtents=<17 8 8>
      constraintSetup {angularLimitsMax=<2.5 40 0>, angularLimitsMin=<-2.5 -40 -20>, angularTarget=<0 0 0>}
      localJointOffset=<3 0 0>
     }]
    planarLimits [{drivingBone {boneName=spine_02}, planeTransform {location=<0 -20 0>, rotation=<rot 0 0 -90>}}]
    relativeSpaceBone {boneName=dyn_tassel}
    sphericalLimits [
     {limitRadius=12, drivingBone {boneName=pelvis}, sphereLocalOffset=<0 0 0>}
     {limitRadius=12, drivingBone {boneName=spine_01}, sphereLocalOffset=<5 0 0>}
     {limitRadius=12, drivingBone {boneName=spine_02}, sphereLocalOffset=<7 0 0>}
    ]
   }
   n19 = Component To Local(ComponentPose=@n18)
   n20 = Get bAnimDynamicsEnabled() -> (bAnimDynamicsEnabled:bool)
   n21 = Get Gravity Override()
   n22 = Get isFemale()
   n23 = RigidBody!(
    ComponentPose=Local To Component(LocalPose=Use cached pose 'Presim'())
    bAlphaBoolEnabled:bool=@n22
    OverrideWorldGravity:Vector=@n21
    ExternalForce
   ) -> (Pose) {
    alphaInputType=Bool
    bClampLinearTranslationLimitToRefPose=true
    bOverrideWorldGravity=true
    overridePhysicsAsset=&/F_MED_IvyCross_Pack_Female_Physics
    alphaBoolBlend {blendOption=Linear}
    baseBoneRef {boneName=root}
    componentAppliedLinearAccClamp=<3000 3000 3000>
    componentLinearAccScale=<0.35 0.45 0.45>
    componentLinearVelScale=<0.35 0.45 0.45>
   }
   n24 = RigidBody(ComponentPose=@n23, bAlphaBoolEnabled=NOT Boolean(A=@n22), OverrideWorldGravity=@n21) {
    alphaInputType=Bool
    bClampLinearTranslationLimitToRefPose=true
    bOverrideWorldGravity=true
    overridePhysicsAsset=&/F_MED_IvyCross_Pack_Male_Physics
    alphaBoolBlend {blendOption=Linear}
    baseBoneRef {boneName=root}
    componentAppliedLinearAccClamp=<3000 3000 3000>
    componentLinearAccScale=<0.3 0.45 0.45>
    componentLinearVelScale=<0.3 0.45 0.45>
   }
   n25 = "Transform (Modify) Bone - Bone: dyn_middle_tassel"!(ComponentPose=@n24, bAlphaBoolEnabled:bool=Get Bending() -> (Bending:bool), Translation="0,0,0", Rotation="0,-15,0", Scale="1,1,1") -> (Pose) {
    alphaInputType=Bool
    rotationMode=BMM_Additive
    rotationSpace=BCS_BoneSpace
    alphaBoolBlend {blendInTime=0.3, blendOption=Linear, blendOutTime=0.4}
    boneToModify {boneName=dyn_middle_tassel}
    rotation {yaw=-15}
   }
   Output Pose(Result=Blend Poses by bool(
     BlendPose_0=Component To Local(ComponentPose=@n25)
     BlendPose_1=Blend Poses by bool(BlendPose_0=@n19, BlendPose_1=@n5, bActiveValue=@n20) {blendType=Linear}
     bActiveValue=Get bRigidBodyEnabled()
    ) {blendType=Linear}) {layerGroup=None}
   n26 = Get isFemale()
   n27 = "Transform (Modify) Bone - Bone: pack"!(ComponentPose=Local To Component(LocalPose=Input Pose()), bAlphaBoolEnabled:bool=@n26, Translation="0,-3,0", Rotation="0,0,0", Scale) -> (Pose) {alphaInputType=Bool, translationMode=BMM_Additive, alphaBoolBlend {blendOption=Linear}, boneToModify {boneName=pack}, translation {y=-3}}
   n28 = "Transform (Modify) Bone - Bone: pack"(ComponentPose=@n27, bAlphaBoolEnabled=NOT Boolean(A=@n26), Translation="0,-5,0", Rotation="0,0,0") {alphaInputType=Bool, translationMode=BMM_Additive, alphaBoolBlend {blendOption=Linear}, boneToModify {boneName=pack}, translation {y=-5}}
   n29 = "Transform (Modify) Bone - Bone: pelvis"(
    ComponentPose=@n28
    Alpha=1
    Translation=Get pelvis_translation() -> (pelvis_translation:Vector)
    Rotation=Get pelvis_rotation() -> (pelvis_rotation:Rotator)
    Scale="1,1,1"
   ) {
    rotationMode=BMM_Replace
    rotationSpace=BCS_WorldSpace
    scaleMode=BMM_Additive
    translationMode=BMM_Replace
    translationSpace=BCS_WorldSpace
    alphaBoolBlend {blendInTime=1, blendOption=Linear, blendOutTime=1}
    boneToModify {boneName=pelvis}
    rotation {roll=20}
   }
   n30 = "Transform (Modify) Bone - Bone: spine_02"(
    ComponentPose=@n29
    Alpha=1
    Translation=Get spine02_translation() -> (spine02_translation:Vector)
    Rotation=Get spine02_rotation() -> (spine02_rotation:Rotator)
    Scale="1,1,1"
   ) {
    rotationMode=BMM_Replace
    rotationSpace=BCS_WorldSpace
    scaleMode=BMM_Additive
    translationMode=BMM_Replace
    translationSpace=BCS_WorldSpace
    alphaBoolBlend {blendInTime=1, blendOption=Linear, blendOutTime=1}
    boneToModify {boneName=spine_02}
    rotation {roll=20}
   }
   n31 = "Transform (Modify) Bone - Bone: spine_01"(
    ComponentPose=@n30
    Alpha=1
    Translation=Get spine01_translation() -> (spine01_translation:Vector)
    Rotation=Get spine01_rotation() -> (spine01_rotation:Rotator)
    Scale="1,1,1"
   ) {
    rotationMode=BMM_Replace
    rotationSpace=BCS_WorldSpace
    scaleMode=BMM_Additive
    translationMode=BMM_Replace
    translationSpace=BCS_WorldSpace
    alphaBoolBlend {blendInTime=1, blendOption=Linear, blendOutTime=1}
    boneToModify {boneName=spine_01}
    rotation {roll=20}
   }
   Save cached pose 'Presim'(Pose=Component To Local(ComponentPose=@n31)) {cacheName=Presim}
   NOT Boolean(A=Get isFemale())
  ]}
 EventGraph {nodes [
   n1 = Event BlueprintUpdateAnimation()
   > n2 = Sequence()
   > Set Delta Time X(Delta Time X=@n1.DeltaTimeX)
   > Is Valid(InputObject=Get BodyMesh() -> (BodyMesh:Skeletal Mesh Component Object Reference))
   n3 = BreakTransform(InTransform=GetSocketTransform(self=Get BodyMesh(), InSocketName=spine_01, TransformSpace=RTS_World))
   > Set spine01_translation(spine01_translation:Vector=@n3) -> (Output_Get:Vector)
   > Set spine01_rotation(spine01_rotation:Rotator=@n3.Rotation) -> (Output_Get:Rotator)
   n4 = BreakTransform(InTransform=GetSocketTransform(self=Get BodyMesh(), InSocketName=spine_02, TransformSpace=RTS_World))
   > Set spine02_translation(spine02_translation:Vector=@n4) -> (Output_Get:Vector)
   > Set spine02_rotation(spine02_rotation:Rotator=@n4.Rotation) -> (Output_Get:Rotator)
   n5 = BreakTransform(InTransform=GetSocketTransform(self=Get BodyMesh(), InSocketName=pelvis, TransformSpace=RTS_World))
   > Set pelvis_translation(pelvis_translation:Vector=@n5) -> (Output_Get:Vector)
   > Set pelvis_rotation(pelvis_rotation:Rotator=@n5.Rotation) -> (Output_Get:Rotator)
   n6 = Get bIsSkyDiving()
   n7 = Get bIsMoving2D()
   > Set Bending(Bending:bool=AND Boolean!(A=Get bIsFalling(), B=NOT Boolean(A=@n6), C:bool=NOT Boolean(A=@n7)) -> (ReturnValue)) -> (Output_Get:bool)
   > n8 = Gravity Override State Function DEPRECATED(CharacterPartAnimInstance=Self-Reference(), GameTimeInSeconds=GetGameTimeInSeconds(), GravityOverrideParams=Get Gravity Override Params())
   n9 = AND Boolean!(A=OR Boolean(A=Get bIsJumping(), B=Get bIsFalling()), B=NOT Boolean(A=Get bIsSkyDiving())) -> (ReturnValue)
   > n10 = Set AlphaJumping(AlphaJumping:float=FInterpTo(Current=Get AlphaJumping() -> (AlphaJumping:float), Target=SelectFloat(A=1, B=0, bPickA=@n9), DeltaTime=Get Delta Time X(), InterpSpeed=8)) -> (Output_Get:float)
   > Set Gravity Override(Gravity Override="Lerp (Vector)"!(A:Vector=@n8, B:Vector="0,0,-1200", Alpha:float=@n10) -> (ReturnValue:Vector))
   n11 = Get bIsSkyDiving()
   > Set Component Linear Acc Scale(Component Linear Acc Scale:Vector=SelectVector(A="0.17,0.45,0.45", B="0.45,0.45,0.45", bPickA=@n11)) -> (Output_Get:Vector)
   > Set Component Linear Vel Scale(Component Linear Vel Scale:Vector=SelectVector(A="0.17,0.45,0.45", B="0.45,0.45,0.45", bPickA=@n11)) -> (Output_Get:Vector)
   n12 = Branch(execute=@n2.then_1, Condition=Get bIsSkyDiving())
   n13 = Get isFemale()
   > Set Is Female Trail Alpha(Is Female Trail Alpha:float=float * float(A="To Float (Boolean)"(InBool=@n13), B=0.5) {is=title K2Node_PromotableOperator}) -> (Output_Get:float)
   n14 = NOT Boolean(A=@n13)
   > Set Is Male Trail Alpha(Is Male Trail Alpha:float=float * float(A="To Float (Boolean)"(InBool=@n14), B=0.5)) -> (Output_Get:float)
   n15 = Get isFemale()
   Set Is Female Trail Alpha(execute=@n12.else, Is Female Trail Alpha="To Float (Boolean)"(InBool=@n15))
   > Set Is Male Trail Alpha(Is Male Trail Alpha="To Float (Boolean)"(InBool=NOT Boolean(A=@n15)))
   Event OnCharacterCustomizationLoaded()
   > n16 = Cast To PlayerPawn_Athena(Object=TryGetPawnOwner())
   > Set As Player Pawn Athena(As Player Pawn Athena:Player Pawn Athena Object Reference=@n16) -> (Output_Get:Player Pawn Athena Object Reference)
   > Set BodyMesh(BodyMesh=GetSkeletalMeshForPartType(self=@n16))
   > n17 = FindGender!(execute, Object:Object Reference=@n16) -> (True:Exec, False:Exec, OutGender:EFortCustomGender Enum=Invalid)
   > Set isFemale(isFemale:bool="Equal (Enum)"(A:EFortCustomGender Enum=@n17, B:EFortCustomGender Enum=Female)) -> (Output_Get:bool)
   TryGetPawnOwner()
  ]}
]
variables [
 AlphaJumping {type=number}
 As Player Pawn Athena {type=PlayerPawn_Athena_C}
 Bend_l {type=number}
 Bending {type=boolean}
 BodyMesh {type=SkeletalMeshComponent}
 Component Linear Acc Scale {type=Vector, default=<0.45 0.45 0.45>}
 Component Linear Vel Scale {type=Vector, default=<0.45 0.45 0.45>}
 Delta Time X {type=number}
 Female_not_skydiving {type=boolean}
 Female_Skydiving {type=boolean}
 Gravity Override {type=Vector}
 Gravity Override Params {
  type=GravityOverrideParamsStruct
  default {
   skydive_DivingDown {gravityOverride=<-1500 0 600>}
   skydive_FloatingDown {gravityOverride=<0 -700 100>, windFrequency=<1.5 4 5>}
   skydive_Parachuting {windFrequency=<1.5 4 5>}
   skydive_Surfing {gravityOverride=<0 0 -980>, windAmplitude=<100 -100 -400>}
  }
 }
 Gravity Override Params_0 {type=GravityOverrideParamsStruct}
 Is Female Trail Alpha {type=number}
 Is Male Trail Alpha {type=number}
 isFemale {type=boolean}
 isMale {type=boolean}
 pelvis_rotation {type=Rotator}
 pelvis_translation {type=Vector}
 spine01_rotation {type=Rotator}
 spine01_translation {type=Vector}
 spine02_rotation {type=Rotator}
 spine02_translation {type=Vector}
]
